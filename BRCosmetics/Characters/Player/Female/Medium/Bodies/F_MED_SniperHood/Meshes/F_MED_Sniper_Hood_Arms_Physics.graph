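# Physics asset for the right-arm scarf cloth on the female medium Sniper Hood character, with box bodies on the scarf bones.
# A spine capsule and right upper-arm capsule and twist box form the body side.
# Front and middle chains hang from the arm twist via driver bones, a back chain from spine_05, and a bkA branch from the middle chain.
# Root and driver joints swing and twist freely, while downstream segments and frills have angular limits with tight twist.
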PhysicsAsset /BRCosmetics/Characters/Player/Female/Medium/Bodies/F_MED_SniperHood/Meshes/F_MED_Sniper_Hood_Arms_Physics
bodies {
 driver_scarf_arm_01 [{shapeType=Box, extentX=1.01, extentY=1.01, extentZ=1.01, rotation=<rot 0 -90 0>}]
 driver_scarf_arm_bkA_01 [{shapeType=Box, extentX=1.01, extentY=1.01, extentZ=1.01, rotation=<rot 0 -90 0>}]
 driver_scarf_arm_bkA_02 [{shapeType=Box, extentX=2.9014, extentY=1.1149, extentZ=2.5504, center=<-1.1087 0 0.0001>, rotation=<rot 0 0 -68.5853>}]
 driver_scarf_arm_bkA_03 [{shapeType=Box, extentX=3.4404, extentY=2.9736, extentZ=1, center=<-1.2401 0 0>, rotation=<rot 0 0 6.2079>}]
 driver_scarf_arm_bk_01 [{shapeType=Box, extentX=1.01, extentY=1.01, extentZ=1.01, rotation=<rot 0 -90 0>}]
 driver_scarf_arm_fr_01 [{shapeType=Box, extentX=1.01, extentY=1.01, extentZ=1.01, rotation=<rot 0 -90 0>}]
 dyn_scarf_arm_01 [{shapeType=Box, extentX=6.1672, extentY=0.5, extentZ=5.9472}]
 dyn_scarf_arm_02 [{shapeType=Box, extentX=4.1541, extentY=4.8188, extentZ=0.5, center=<-1.8984 -0.0001 0>}]
 dyn_scarf_arm_03 [{shapeType=Box, extentX=4.01, extentY=3, extentZ=1.5, center=<-1.3821 0 -0.522>}]
 dyn_scarf_arm_bkA_01 [{shapeType=Box, extentX=4.3933, extentY=0.9221, extentZ=6.6132, center=<-3.2166 0 0.0001>, rotation=<rot 0 90 90>}]
 dyn_scarf_arm_bk_01 [{shapeType=Box, extentX=1.01, extentY=1.01, extentZ=1.01, rotation=<rot 0 -90 0>}]
 dyn_scarf_arm_bk_02 [{shapeType=Box, extentX=3.2956, extentY=2.9608, extentZ=1.1789, center=<-1.3027 0 0>}]
 dyn_scarf_arm_bk_03 [{shapeType=Box, extentX=3.7217, extentY=2.9758, extentZ=0.7818, center=<-1.6687 0 0>}]
 dyn_scarf_arm_fr_01 [{shapeType=Box, extentX=7.7062, extentY=0.5, extentZ=9.3439, center=<-4.3254 0 0>, rotation=<rot 0 90 90>}]
 dyn_scarf_arm_fr_02 [{shapeType=Box, extentX=3.51, extentY=4.51, extentZ=0.5, center=<-1.7247 0 0.0247>, rotation=<rot 0 -90 0>}]
 dyn_scarf_arm_fr_03 [{shapeType=Box, extentX=4, extentY=0.5467, extentZ=3, center=<-2.017 0.1385 -0.5724>, rotation=<rot 0.9703 3.6345 -84.4878>}]
 dyn_scarf_frillA_01 [{shapeType=Box, extentX=3, extentY=3.01, extentZ=1.01, center=<-1.2623 0.0857 0.116>, rotation=<rot 0 -90 0>}]
 dyn_scarf_frillB_02 [{shapeType=Box, extentX=3, extentY=3.01, extentZ=1.01, center=<-1.2705 0 0.0001>, rotation=<rot 0 -90 0>}]
 dyn_scarf_frillC_01 [{shapeType=Box, extentX=4.5917, extentY=0.5445, extentZ=3.2573, center=<-1.926 -0.0582 -0.1181>, rotation=<rot 0 7.22 -88.8326>}]
 dyn_scarf_frillH_01 [{shapeType=Box, extentX=3.6104, extentY=3.4646, extentZ=0.5, center=<-1.356 0 0>}]
 spine_05 [{shapeType=Capsule, length=7.0892, radius=11.4765}]
 upperarm_r [{shapeType=Capsule, length=21.3777, radius=4.4966, center=<-13.8961 0.324 0.6727>, rotation=<rot 0 90 90>}]
 upperarm_twist_01_r [{shapeType=Box, extentX=3.7097, extentY=1.545, extentZ=3.3404}]
}
constraints [
 {bone1Name=dyn_scarf_arm_fr_02, bone2Name=dyn_scarf_arm_fr_01, swing1LimitDegrees=45, swing2LimitDegrees=20, twistLimitDegrees=5}
 {bone1Name=dyn_scarf_arm_fr_03, bone2Name=dyn_scarf_arm_fr_02, swing1LimitDegrees=45, swing2LimitDegrees=20, twistLimitDegrees=5}
 {
  bone1Name=upperarm_r
  bone2Name=spine_05
  swing1LimitDegrees=45
  swing1Motion=Free
  swing2LimitDegrees=45
  swing2Motion=Free
  twistLimitDegrees=45
  twistMotion=Free
 }
 {
  bone1Name=dyn_scarf_arm_fr_01
  bone2Name=driver_scarf_arm_fr_01
  swing1LimitDegrees=45
  swing1Motion=Free
  swing2LimitDegrees=45
  swing2Motion=Free
  twistLimitDegrees=45
  twistMotion=Free
 }
 {
  bone1Name=driver_scarf_arm_fr_01
  bone2Name=upperarm_twist_01_r
  swing1LimitDegrees=45
  swing1Motion=Free
  swing2LimitDegrees=45
  swing2Motion=Free
  twistLimitDegrees=45
  twistMotion=Free
 }
 {bone1Name=dyn_scarf_frillA_01, bone2Name=dyn_scarf_arm_fr_01, swing1LimitDegrees=45, swing2LimitDegrees=20, twistLimitDegrees=5}
 {bone1Name=dyn_scarf_frillB_02, bone2Name=dyn_scarf_arm_fr_01, swing1LimitDegrees=45, swing2LimitDegrees=20, twistLimitDegrees=5}
 {
  bone1Name=upperarm_twist_01_r
  bone2Name=upperarm_r
  swing1LimitDegrees=45
  swing1Motion=Free
  swing2LimitDegrees=45
  swing2Motion=Free
  twistLimitDegrees=45
  twistMotion=Free
 }
 {
  bone1Name=driver_scarf_arm_01
  bone2Name=upperarm_twist_01_r
  swing1LimitDegrees=45
  swing1Motion=Free
  swing2LimitDegrees=45
  swing2Motion=Free
  twistLimitDegrees=45
  twistMotion=Free
 }
 {
  bone1Name=dyn_scarf_arm_01
  bone2Name=driver_scarf_arm_01
  swing1LimitDegrees=45
  swing1Motion=Free
  swing2LimitDegrees=45
  swing2Motion=Free
  twistLimitDegrees=45
  twistMotion=Free
 }
 {bone1Name=dyn_scarf_arm_02, bone2Name=dyn_scarf_arm_01, swing1LimitDegrees=45, swing2LimitDegrees=25, twistLimitDegrees=5}
 {bone1Name=dyn_scarf_arm_03, bone2Name=dyn_scarf_arm_02, swing1LimitDegrees=45, swing2LimitDegrees=60, twistLimitDegrees=5}
 {bone1Name=dyn_scarf_frillC_01, bone2Name=dyn_scarf_arm_01, swing1LimitDegrees=45, swing2LimitDegrees=30, twistLimitDegrees=5}
 {bone1Name=dyn_scarf_frillH_01, bone2Name=dyn_scarf_arm_01, swing1LimitDegrees=45, swing2LimitDegrees=30, twistLimitDegrees=5}
 {
  bone1Name=driver_scarf_arm_bkA_01
  bone2Name=dyn_scarf_arm_01
  swing1LimitDegrees=45
  swing1Motion=Free
  swing2LimitDegrees=45
  swing2Motion=Free
  twistLimitDegrees=45
  twistMotion=Free
 }
 {
  bone1Name=dyn_scarf_arm_bkA_01
  bone2Name=driver_scarf_arm_bkA_01
  swing1LimitDegrees=45
  swing1Motion=Free
  swing2LimitDegrees=45
  swing2Motion=Free
  twistLimitDegrees=45
  twistMotion=Free
 }
 {bone1Name=driver_scarf_arm_bkA_02, bone2Name=dyn_scarf_arm_bkA_01, swing1LimitDegrees=45, swing2LimitDegrees=30, twistLimitDegrees=5}
 {bone1Name=driver_scarf_arm_bkA_03, bone2Name=driver_scarf_arm_bkA_02, swing1LimitDegrees=45, swing2LimitDegrees=30, twistLimitDegrees=5}
 {
  bone1Name=driver_scarf_arm_bk_01
  bone2Name=spine_05
  swing1LimitDegrees=45
  swing1Motion=Free
  swing2LimitDegrees=45
  swing2Motion=Free
  twistLimitDegrees=45
  twistMotion=Free
 }
 {
  bone1Name=dyn_scarf_arm_bk_01
  bone2Name=driver_scarf_arm_bk_01
  swing1LimitDegrees=45
  swing1Motion=Free
  swing2LimitDegrees=45
  swing2Motion=Free
  twistLimitDegrees=45
  twistMotion=Free
 }
 {bone1Name=dyn_scarf_arm_bk_02, bone2Name=dyn_scarf_arm_bk_01, swing1LimitDegrees=45, swing2LimitDegrees=35, twistLimitDegrees=5}
 {bone1Name=dyn_scarf_arm_bk_03, bone2Name=dyn_scarf_arm_bk_02, swing1LimitDegrees=45, swing2LimitDegrees=35, twistLimitDegrees=5}
]
