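# Physics asset for the frilly variant of the female medium Sniper Hood cosmetic, covering its dynamic back scarf and frill bones.
# Four small frill boxes are constrained to a larger scarf box, which is constrained to a tiny spine_03 capsule, with wide swing and narrow twist angles.
# A large offset spine_01 capsule supplies torso volume and links to spine_03 through a joint with free swing and twist.
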PhysicsAsset /BRCosmetics/Characters/Player/Female/Medium/Bodies/F_MED_SniperHood/Meshes/F_MED_Sniper_Hood_Frilly_Physics
bodies {
 dyn_frill_bkA_01 [{shapeType=Box, extentX=4.6665, extentY=3.6686, extentZ=0.5, center=<-1.9186 0 0>}]
 dyn_frill_bkB_01 [{shapeType=Box, extentX=2.51, extentY=2.51, extentZ=0.5, center=<-0.9992 0 0>}]
 dyn_frill_bkC_01 [{shapeType=Box, extentX=4.51, extentY=3.51, extentZ=1.01, center=<-1.8473 0 0.0434>, rotation=<rot -5 0 0>}]
 dyn_frill_bkD_01 [{shapeType=Box, extentX=4.135, extentY=2.3239, extentZ=0.5, center=<-1.8398 0 0>}]
 dyn_scarf_bk_01 [{shapeType=Box, extentX=9, extentY=18, extentZ=2.5, center=<-4.1283 0 -0.0001>}]
 spine_01 [{shapeType=Capsule, length=9.5, radius=22, center=<5.8597 12.2714 -0.0598>, rotation=<rot 0 -90 90>}]
 spine_03 [{shapeType=Capsule, length=0.505, radius=0.505, rotation=<rot 0 -90 0>}]
}
constraints [
 {bone1Name=dyn_frill_bkA_01, bone2Name=dyn_scarf_bk_01, swing1LimitDegrees=45, swing2LimitDegrees=60, twistLimitDegrees=5}
 {bone1Name=dyn_frill_bkB_01, bone2Name=dyn_scarf_bk_01, swing1LimitDegrees=45, swing2LimitDegrees=60, twistLimitDegrees=5}
 {bone1Name=dyn_frill_bkC_01, bone2Name=dyn_scarf_bk_01, swing1LimitDegrees=45, swing2LimitDegrees=60, twistLimitDegrees=5}
 {bone1Name=dyn_frill_bkD_01, bone2Name=dyn_scarf_bk_01, swing1LimitDegrees=45, swing2LimitDegrees=47.4286, twistLimitDegrees=5}
 {
  bone1Name=spine_03
  bone2Name=spine_01
  swing1LimitDegrees=45
  swing1Motion=Free
  swing2LimitDegrees=45
  swing2Motion=Free
  twistLimitDegrees=45
  twistMotion=Free
 }
 {bone1Name=dyn_scarf_bk_01, bone2Name=spine_03, swing1LimitDegrees=45, swing2LimitDegrees=90, twistLimitDegrees=2}
]
