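# Post-process-style cosmetic AnimBP for the female Sniper Hood body, copying the parent mesh pose and adding secondary motion to scarf, shoulder pad and boot strings.
# bIsSkyDiving switches between shoulder-pad and scarf Pose Drivers, feeding weighted constraints and additive bone offsets driven by right upper-arm yaw.
# AnimDynamics moves boot strings, then two RigidBody nodes simulate frills and arms.
# Frill physics is disabled when a cape or backbling is detected.
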AnimBlueprint /BRCosmetics/Characters/Player/Female/Medium/Bodies/F_MED_SniperHood/Meshes/F_MED_Sniper_Hood_AnimBP
target_skeleton=&/F_MED_Sniper_Hood_Skeleton
graphs [
 AnimGraph {nodes [
   n1 = Pose Driver - Source: upperarm_l!(SourcePose:Pose Link Structure=Use cached pose 'AnimCache'()) -> (Pose:Pose Link Structure) {
    is=title AnimGraphNode_PoseDriver
    poseAsset=&/sniperhood_shoulderpad_PoseAsset
    onlyDriveBones [{boneName=dyn_shldrpad_l}]
    poseTargets [
     {
      bApplyCustomCurve=false
      bIsHidden=false
      distanceMethod=DefaultMethod
      drivenName=Pose_7
      functionType=DefaultFunction
      targetScale=1
      boneTransforms [
       {targetRotation=<rot -62.0413 -0.1866 1.0464>, targetTranslation=<9.8658 -1.2136 -0.4105>}
       {targetRotation=<rot -85.1913 -138.5624 143.0618>, targetTranslation=<5.4831 -0.1053 -3.2946>}
      ]
      customCurve {defaultValue=3.4028234663852886e+38, postInfinityExtrap=RCCE_Constant, preInfinityExtrap=RCCE_Constant, keys []}
      targetRotation=<rot 0 0 0>
     }
     {
      bApplyCustomCurve=false
      bIsHidden=false
      distanceMethod=DefaultMethod
      drivenName=Pose_12
      functionType=DefaultFunction
      targetScale=0.5447
      boneTransforms [
       {targetRotation=<rot 5.2274 6.0278 -72.2107>, targetTranslation=<9.8658 -1.2136 -0.4105>}
       {targetRotation=<rot -64.8581 -92.971 99.1839>, targetTranslation=<5.483 -0.1053 -3.2946>}
      ]
      customCurve {defaultValue=3.4028234663852886e+38, postInfinityExtrap=RCCE_Constant, preInfinityExtrap=RCCE_Constant, keys []}
      targetRotation=<rot 0 0 0>
     }
     {
      bApplyCustomCurve=false
      bIsHidden=false
      distanceMethod=DefaultMethod
      drivenName=Pose_13
      functionType=DefaultFunction
      targetScale=0.5447
      boneTransforms [
       {targetRotation=<rot 8.4368 -10.7445 -84.4428>, targetTranslation=<9.8658 -1.2135 -0.4105>}
       {targetRotation=<rot -58.2253 25.622 -16.981>, targetTranslation=<5.4829 -0.1052 -3.2946>}
      ]
      customCurve {defaultValue=3.4028234663852886e+38, postInfinityExtrap=RCCE_Constant, preInfinityExtrap=RCCE_Constant, keys []}
      targetRotation=<rot 0 0 0>
     }
    ]
    rBFParams {distanceMethod=Quaternion, radius=48.2319}
    sourceBones [{boneName=upperarm_l}, {boneName=clavicle_l}]
   }
   n2 = Pose Driver - Source: clavicle_r!(SourcePose:Pose Link Structure=Use cached pose 'AnimCache'()) -> (Pose:Pose Link Structure) {
    is=title AnimGraphNode_PoseDriver
    poseAsset=&/F_MED_Sniper_Hood_Cloth_Poses_PoseAsset
    onlyDriveBones [{boneName=driver_scarf_arm_01}, {boneName=dyn_uppeararm_twist_01_r}, {boneName=driver_scarf_arm_fr_01}, {boneName=driver_scarf_arm_bkA_01}]
    poseTargets [
     {
      bApplyCustomCurve=false
      bIsHidden=false
      distanceMethod=DefaultMethod
      drivenName=Base
      functionType=DefaultFunction
      targetScale=0.5203
      boneTransforms [
       {targetRotation=<rot -85.1914 41.4375 143.0618>, targetTranslation=<5.4831 -0.1053 3.2946>}
       {targetRotation=<rot -58.1985 2.4027 1.3901>, targetTranslation=<-9.8658 1.2136 0.4104>}
      ]
      customCurve {defaultValue=3.4028234663852886e+38, postInfinityExtrap=RCCE_Constant, preInfinityExtrap=RCCE_Constant, keys []}
      targetRotation=<rot 0 0 0>
     }
     {
      bApplyCustomCurve=false
      bIsHidden=false
      distanceMethod=DefaultMethod
      drivenName=Pose_1
      functionType=DefaultFunction
      targetScale=0.4142
      boneTransforms [
       {targetRotation=<rot -85.1914 41.4375 143.0618>, targetTranslation=<5.4831 -0.1053 3.2946>}
       {targetRotation=<rot -34.9855 35.1604 21.5717>, targetTranslation=<-9.8054 1.6418 -0.3652>}
      ]
      customCurve {defaultValue=3.4028234663852886e+38, postInfinityExtrap=RCCE_Constant, preInfinityExtrap=RCCE_Constant, keys []}
      targetRotation=<rot 0 0 0>
     }
     {
      bApplyCustomCurve=false
      bIsHidden=false
      distanceMethod=DefaultMethod
      drivenName=Pose_2
      functionType=DefaultFunction
      targetScale=0.2464
      boneTransforms [
       {targetRotation=<rot -85.1914 41.4375 143.0618>, targetTranslation=<5.4831 -0.1053 3.2946>}
       {targetRotation=<rot -46.0833 91.2657 -36.856>, targetTranslation=<-9.5741 0.9221 -2.541>}
      ]
      customCurve {defaultValue=3.4028234663852886e+38, postInfinityExtrap=RCCE_Constant, preInfinityExtrap=RCCE_Constant, keys []}
      targetRotation=<rot 0 0 0>
     }
     {
      bApplyCustomCurve=true
      bIsHidden=false
      distanceMethod=DefaultMethod
      drivenName=Pose_3
      functionType=DefaultFunction
      targetScale=1
      boneTransforms [
       {targetRotation=<rot -85.1914 41.4375 143.0618>, targetTranslation=<5.4831 -0.1053 3.2946>}
       {targetRotation=<rot 30.709 115.269 -59.5451>, targetTranslation=<-9.0403 0.1113 -4.1487>}
      ]
      customCurve {
       defaultValue=3.4028234663852886e+38
       postInfinityExtrap=RCCE_Constant
       preInfinityExtrap=RCCE_Constant
       keys [
        {
         arriveTangent=0
         arriveTangentWeight=0
         interpMode=RCIM_Linear
         leaveTangent=0
         leaveTangentWeight=0
         tangentMode=RCTM_Auto
         tangentWeightMode=RCTWM_WeightedNone
         time=0
         value=0
        }
        {
         arriveTangent=0
         arriveTangentWeight=0
         interpMode=RCIM_Linear
         leaveTangent=0
         leaveTangentWeight=0
         tangentMode=RCTM_Auto
         tangentWeightMode=RCTWM_WeightedNone
         time=0.63
         value=0.0308
        }
        {
         arriveTangent=0
         arriveTangentWeight=0
         interpMode=RCIM_Linear
         leaveTangent=0
         leaveTangentWeight=0
         tangentMode=RCTM_Auto
         tangentWeightMode=RCTWM_WeightedNone
         time=0.8521
         value=0.9933
        }
       ]
      }
      targetRotation=<rot 0 0 0>
     }
     {
      bApplyCustomCurve=false
      bIsHidden=false
      distanceMethod=DefaultMethod
      drivenName=Pose_4
      functionType=DefaultFunction
      targetScale=0.2464
      boneTransforms [
       {targetRotation=<rot -85.1914 41.4375 143.0618>, targetTranslation=<5.4831 -0.1053 3.2946>}
       {targetRotation=<rot -49.8566 92.8085 -18.28>, targetTranslation=<-9.3136 -0.1097 -3.4953>}
      ]
      customCurve {defaultValue=3.4028234663852886e+38, postInfinityExtrap=RCCE_Constant, preInfinityExtrap=RCCE_Constant, keys []}
      targetRotation=<rot 0 0 0>
     }
     {
      bApplyCustomCurve=false
      bIsHidden=false
      distanceMethod=DefaultMethod
      drivenName=Pose_5
      functionType=DefaultFunction
      targetScale=0.4142
      boneTransforms [
       {targetRotation=<rot -85.1914 41.4375 143.0618>, targetTranslation=<5.4831 -0.1053 3.2946>}
       {targetRotation=<rot -65.1849 11.4163 35.123>, targetTranslation=<-9.7822 -1.7298 -0.5478>}
      ]
      customCurve {defaultValue=3.4028234663852886e+38, postInfinityExtrap=RCCE_Constant, preInfinityExtrap=RCCE_Constant, keys []}
      targetRotation=<rot 0 0 0>
     }
     {
      bApplyCustomCurve=true
      bIsHidden=false
      distanceMethod=DefaultMethod
      drivenName=Pose_6
      functionType=DefaultFunction
      targetScale=0.5663
      boneTransforms [
       {targetRotation=<rot -85.1914 41.4375 143.0618>, targetTranslation=<5.4831 -0.1053 3.2946>}
       {targetRotation=<rot -0.744 1.8594 32.7185>, targetTranslation=<-6.4312 3.3585 -6.7846>}
      ]
      customCurve {
       defaultValue=3.4028234663852886e+38
       postInfinityExtrap=RCCE_Constant
       preInfinityExtrap=RCCE_Constant
       keys [
        {
         arriveTangent=0
         arriveTangentWeight=0
         interpMode=RCIM_Linear
         leaveTangent=0
         leaveTangentWeight=0
         tangentMode=RCTM_Auto
         tangentWeightMode=RCTWM_WeightedNone
         time=0
         value=0
        }
        {
         arriveTangent=0
         arriveTangentWeight=0
         interpMode=RCIM_Linear
         leaveTangent=0
         leaveTangentWeight=0
         tangentMode=RCTM_Auto
         tangentWeightMode=RCTWM_WeightedNone
         time=1
         value=1
        }
       ]
      }
      targetRotation=<rot 0 0 0>
     }
    ]
    rBFParams {distanceMethod=Quaternion, radius=64}
    sourceBones [{boneName=clavicle_r}, {boneName=upperarm_r}]
   }
   n3 = Get bIsSkyDiving()
   n4 = "Constraint - driver_scarf_arm_bk_01 (PPP)"!(
    ComponentPose=Local To Component(LocalPose=Blend Poses by bool(BlendPose_0=@n1, BlendPose_1=@n2, bActiveValue=@n3) {blendType=Linear})
    Alpha=1
    ConstraintWeights_0=0.35
    ConstraintWeights_1:float=0.3
    ConstraintWeights_2:float=0.35
   ) -> (Pose) {
    alphaBoolBlend {blendOption=Linear}
    boneToModify {boneName=driver_scarf_arm_bk_01}
    constraintSetup [{targetBone {boneName=dyn_scarf_arm_01}}, {targetBone {boneName=spine_05}}, {targetBone {boneName=clavicle_r}}]
    constraintWeights [1,1,1]
   }
   n5 = "Constraint - driver_scarf_arm_bkA_01 (PPP)"!(ComponentPose=@n4, Alpha=1, ConstraintWeights_0=0.9, ConstraintWeights_1:float=0.15, ConstraintWeights_2:float=0.25) -> (Pose) {
    alphaBoolBlend {blendOption=Linear}
    boneToModify {boneName=driver_scarf_arm_bkA_01}
    constraintSetup [{targetBone {boneName=dyn_scarf_arm_01}}, {targetBone {boneName=spine_05}}, {targetBone {boneName=clavicle_r}}]
    constraintWeights [1,1,1]
   }
   n6 = "Transform (Modify) Bone - Bone: driver_scarf_arm_bkA_02"(ComponentPose=@n5, Alpha=Get arm_up_r() -> (arm_up_r:double), Translation="0,0,0", Rotation="-1.866,4.997,89.320", Scale="1,1,1") {
    rotationMode=BMM_Additive
    rotationSpace=BCS_BoneSpace
    translationMode=BMM_Additive
    alphaBoolBlend {blendOption=Linear}
    boneToModify {boneName=driver_scarf_arm_bkA_02}
    rotation=<rot -1.866 4.9971 89.32>
   }
   n7 = Get arm_back_r() -> (arm_back_r:double)
   n8 = "Transform (Modify) Bone - Bone: driver_scarf_arm_bkA_01"(ComponentPose=@n6, Alpha=@n7, Translation="2.753,-1.968,-0.029", Rotation="1.824,5.247,0.348", Scale="1,1,1") {
    rotationMode=BMM_Additive
    rotationSpace=BCS_ParentBoneSpace
    translationMode=BMM_Additive
    alphaBoolBlend {blendOption=Linear}
    boneToModify {boneName=driver_scarf_arm_bkA_01}
    rotation=<rot 1.824 5.2471 0.348>
    translation=<2.753 -1.968 -0.029>
   }
   n9 = "Transform (Modify) Bone - Bone: driver_scarf_arm_bk_01"(ComponentPose=@n8, Alpha=@n7, Translation="0,0,0", Rotation="0,0,5", Scale="1,1,1") {
    rotationMode=BMM_Additive
    rotationSpace=BCS_ParentBoneSpace
    translationMode=BMM_Additive
    alphaBoolBlend {blendOption=Linear}
    boneToModify {boneName=driver_scarf_arm_bk_01}
    rotation {roll=5}
   }
   n10 = Anim Dynamics - Bone: dyn_boot_strings_01_l(ComponentPose=@n9, Alpha=0.55, ExternalForce="0,0,0") {
    bGravityOverrideInSimSpace=true
    bLinearSpring=true
    bOverrideLinearDamping=true
    gravityScale=0.33
    linearDampingOverride=0.995
    linearSpringConstant=150
    simulationSpace=RootRelative
    alphaBoolBlend {blendOption=Linear}
    boundBone {boneName=dyn_boot_strings_01_l}
    physicsBodyDefinitions [{
      sphereCollisionRadius=0
      boundBone {boneName=dyn_boot_strings_01_l}
      boxExtents=<10 10 10>
      constraintSetup {angularLimitsMax=<0 0 0>, angularLimitsMin=<0 0 0>, angularTarget=<0 0 0>, linearAxesMax=<4 2 0.02>, linearAxesMin=<-4 -2 -0.02>}
      localJointOffset=<0 0 -5>
     }]
   }
   n11 = Anim Dynamics - Bone: dyn_boot_strings_02_l(ComponentPose=@n10, Alpha=0.3, ExternalForce="0,0,0") {
    bGravityOverrideInSimSpace=true
    bLinearSpring=true
    bOverrideLinearDamping=true
    gravityScale=0.29
    linearDampingOverride=0.995
    linearSpringConstant=150
    simulationSpace=RootRelative
    alphaBoolBlend {blendOption=Linear}
    boundBone {boneName=dyn_boot_strings_02_l}
    physicsBodyDefinitions [{
      sphereCollisionRadius=0
      boundBone {boneName=dyn_boot_strings_02_l}
      boxExtents=<10 10 10>
      constraintSetup {angularLimitsMax=<0 0 0>, angularLimitsMin=<0 0 0>, angularTarget=<0 0 0>, linearAxesMax=<4 2 0.02>, linearAxesMin=<-5 -2 -0.02>}
      localJointOffset=<0 0 -6>
     }]
   }
   n12 = Anim Dynamics - Bone: dyn_boot_strings_01_r(ComponentPose=@n11, Alpha=0.55, ExternalForce="0,0,0") {
    bGravityOverrideInSimSpace=true
    bLinearSpring=true
    bOverrideLinearDamping=true
    gravityScale=0.33
    linearDampingOverride=0.995
    linearSpringConstant=150
    simulationSpace=RootRelative
    alphaBoolBlend {blendOption=Linear}
    boundBone {boneName=dyn_boot_strings_01_r}
    physicsBodyDefinitions [{
      sphereCollisionRadius=0
      boundBone {boneName=dyn_boot_strings_01_r}
      boxExtents=<10 10 10>
      constraintSetup {angularLimitsMax=<0 0 0>, angularLimitsMin=<0 0 0>, angularTarget=<0 0 0>, linearAxesMax=<4 2 0.02>, linearAxesMin=<-4 -2 -0.02>}
      localJointOffset=<0 0 -5>
     }]
   }
   n13 = Anim Dynamics - Bone: dyn_boot_strings_02_r(ComponentPose=@n12, Alpha=0.3, ExternalForce="0,0,0") {
    bGravityOverrideInSimSpace=true
    bLinearSpring=true
    bOverrideLinearDamping=true
    gravityScale=0.29
    linearDampingOverride=0.995
    linearSpringConstant=150
    simulationSpace=RootRelative
    alphaBoolBlend {blendOption=Linear}
    boundBone {boneName=dyn_boot_strings_02_r}
    physicsBodyDefinitions [{
      sphereCollisionRadius=0
      boundBone {boneName=dyn_boot_strings_02_r}
      boxExtents=<10 10 10>
      constraintSetup {angularLimitsMax=<0 0 0>, angularLimitsMin=<0 0 0>, angularTarget=<0 0 0>, linearAxesMax=<4 2 0.02>, linearAxesMin=<-5 -2 -0.02>}
      localJointOffset=<0 0 -6>
     }]
   }
   n14 = "Constraint - dyn_scarf_frillA_01 (PPPPP)"!(
    ComponentPose=@n13
    Alpha=1
    ConstraintWeights_0=0.4073
    ConstraintWeights_1:float=0.039
    ConstraintWeights_2:float=0.1223
    ConstraintWeights_3:float=0.1257
    ConstraintWeights_4:float=0.3057
   ) -> (Pose) {
    alphaBoolBlend {blendOption=Linear}
    boneToModify {boneName=dyn_scarf_frillA_01}
    constraintSetup [
     {targetBone {boneName=dyn_scarf_arm_fr_01}}
     {targetBone {boneName=spine_02}}
     {targetBone {boneName=spine_04}}
     {targetBone {boneName=spine_05}}
     {targetBone {boneName=dyn_scarf_arm_01}}
    ]
    constraintWeights [1,1,1,1,1]
   }
   n15 = "Constraint - dyn_scarf_frillB_02 (PPPP)"!(
    ComponentPose=@n14
    Alpha=1
    ConstraintWeights_0=0.4351
    ConstraintWeights_1:float=0.0985
    ConstraintWeights_2:float=0.0526
    ConstraintWeights_3:float=0.4138
   ) -> (Pose) {
    alphaBoolBlend {blendOption=Linear}
    boneToModify {boneName=dyn_scarf_frillB_02}
    constraintSetup [
     {targetBone {boneName=dyn_scarf_arm_fr_01}}
     {targetBone {boneName=spine_04}}
     {targetBone {boneName=spine_05}}
     {targetBone {boneName=dyn_scarf_arm_01}}
    ]
    constraintWeights [1,1,1,1]
   }
   n16 = RigidBody(ComponentPose=@n15, Alpha=Get back_dyn_alpha() -> (back_dyn_alpha:double)) {
    bOverrideWorldGravity=true
    overridePhysicsAsset=&/F_MED_Sniper_Hood_Frilly_Physics
    simulationSpace=BaseBoneSpace
    alphaBoolBlend {blendOption=Linear}
    alphaScaleBiasClamp {bInterpResult=true}
    baseBoneRef {boneName=pelvis}
    componentAppliedLinearAccClamp=<1000 1000 1000>
    componentLinearAccScale=<1 1 1>
    componentLinearVelScale=<1 1 1>
    overrideWorldGravity {z=-980}
   }
   n17 = RigidBody(ComponentPose=@n16, Alpha=0.7) {
    bOverrideWorldGravity=true
    overridePhysicsAsset=&/F_MED_Sniper_Hood_Arms_Physics
    simulationSpace=BaseBoneSpace
    alphaBoolBlend {blendOption=Linear}
    baseBoneRef {boneName=root}
    componentAppliedLinearAccClamp=<0.1 0.1 0.1>
    componentLinearAccScale=<1 1 1>
    componentLinearVelScale=<10 10 10>
    overrideWorldGravity {z=-980}
   }
   Output Pose(Result=Component To Local(ComponentPose=@n17)) {layerGroup=None}
   Save cached pose 'AnimCache'(Pose=Copy Pose From Mesh() {bCopyCurves=true, bUseAttachedParent=true}) {cacheName=AnimCache}
  ]}
 EventGraph {nodes [
   Event BlueprintUpdateAnimation()
   > n1 = Sequence!(execute) -> (then_0, then_1, then_2:Exec)
   Do Once(execute=@n1.then_1)
   n2 = GetPartSkeletalMeshComponent()
   n3 = "Not Equal (Object)"!(
    A:Object Reference=FindChildCommponentFromAttachSocket!(
     self:Custom Character Part Anim Instance Object Reference
     Parent:Skeletal Mesh Component Object Reference=@n2
     bIncludeAllDescendants:bool=false
     AttachSocket:Name=Cape
    ) -> (ReturnValue:Skeletal Mesh Component Object Reference)
    B:Object Reference
   ) -> (ReturnValue:bool)
   n4 = "Not Equal (Object)"(A=FindChildCommponentFromAttachSocket(Parent=@n2, bIncludeAllDescendants=false, AttachSocket=backpack_br))
   > Set back_dyn_alpha(back_dyn_alpha:double=SelectFloat(A=0, B=1, bPickA=OR Boolean(A=@n3, B=@n4))) -> (Output_Get:double)
   n5 = GetOwnerSkeletalMeshComponent()
   n6 = BreakRotator(InRot=BreakTransform(InTransform=GetDeltaTransformFromRefPose(self=@n5, BoneName=upperarm_r, BaseName=clavicle_l)).Rotation)
   n7 = MapRangeClamped(Value=@n6.Yaw, InRangeA=0, InRangeB=60)
   Set upperarm_rotate_r(execute=@n1.then_0, upperarm_rotate_r:Rotator=MakeRotator(Yaw=@n7)) -> (Output_Get:Rotator)
   > Set arm_up_r(arm_up_r:double=@n7) -> (Output_Get:double)
   > Set arm_back_r(arm_back_r:double=MapRangeClamped(Value=@n6.Yaw, InRangeA=35, InRangeB=-20)) -> (Output_Get:double)
   Event OnCharacterCustomizationLoaded()
   > Do Once()
   n8 = CheckForBackbling()
   > Set back_dyn_alpha(back_dyn_alpha=SelectFloat(A=0, B=1, bPickA=OR Boolean(A=@n8, B=@n8.bIsBackblingLong)))
   TryGetPawnOwner()
  ]}
]
variables [
 arm_back_r {type=number}
 arm_up_r {type=number}
 back_dyn_alpha {type=number}
 upperarm_rotate_r {type=Rotator}
]
