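# Physics asset for the hood mesh of the medium female Sniper Hood player cosmetic body, covering the upper torso and right shoulder.
# Capsule bodies sit on spine_05 (two capsules), clavicle_r and upperarm_twist_01_r, with no left-side bodies.
# Constraints chain upper arm twist to clavicle to spine, with all swing and twist motion free, so they impose no angular limits.
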PhysicsAsset /BRCosmetics/Characters/Player/Female/Medium/Bodies/F_MED_SniperHood/Meshes/F_MED_Sniper_Hood_Physics
bodies {
 clavicle_r [{shapeType=Capsule, length=9.2433, radius=6.3726, center=<-5.17 -0.1878 2.5785>, rotation=<rot 90 0 0>}]
 spine_05 [
  {shapeType=Capsule, length=11.1449, radius=11.5242, center=<-10.3359 -0.8687 1.1482>, rotation=<rot 0.6147 93.6495 -63.67>}
  {shapeType=Capsule, length=12.5, radius=12.25, center=<-5.9797 2.278 0.0831>, rotation=<rot 0 0 -5.0001>}
 ]
 upperarm_twist_01_r [{shapeType=Capsule, length=17.0729, radius=5.8981, center=<-5.3041 -0.2891 -0.5767>, rotation=<rot 90 0 0>}]
}
constraints [
 {
  bone1Name=upperarm_twist_01_r
  bone2Name=clavicle_r
  swing1LimitDegrees=45
  swing1Motion=Free
  swing2LimitDegrees=45
  swing2Motion=Free
  twistLimitDegrees=45
  twistMotion=Free
 }
 {
  bone1Name=clavicle_r
  bone2Name=spine_05
  swing1LimitDegrees=45
  swing1Motion=Free
  swing2LimitDegrees=45
  swing2Motion=Free
  twistLimitDegrees=45
  twistMotion=Free
 }
]
